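; Gameplay ability for a Battle Pass S20 quest that counts mantles and reports progress once five stacks accumulate.
; Each gameplay-event activation calls the parent, applies a counter effect to the owner, stores its handle and prints the active stack count.
; At five or more stacks it sends the Athena.Quests.MantleXTimes stat event to the owning player's quest manager, removes the effect and ends.
; Below that threshold this graph never calls EndAbility.
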(event Ability|EventActivateAbilityFromEvent
  (params (EventData :type Gameplay%20Event%20Data%20Structure%20%28by%20ref%29))
  (|CallParentFunction EventData)
  (bind _returnvalue (Ability|ApplyGameplayEffectToOwner (Variables|Default|GetGameplay%20Effect%20Class)))
  (bind _output_get (Variables|Default|SetGE%20Handle _returnvalue))
  (bind _returnvalue_2 (Ability|GameplayEffect|GetActiveGameplayEffectStackCount _output_get))
  (Development|PrintString (Utilities|String|ToString(Integer) _returnvalue_2))
  (if (>= _returnvalue_2 5)
    (Quest|SendComplexCustomStatEvent (Variables|GAVars|Data|GetOwningPlayerQuestManager) 0 "(GameplayTags=((TagName=\"Athena.Quests.MantleXTimes\")))" "(GameplayTags=)" 1)
    (Ability|RemoveGameplayEffectFromOwnerWithHandle (Variables|Default|GetGE%20Handle))
    (Ability|EndAbility)))
