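; Passive quest-tracker ability that reports progress for the Battle Pass Season 20 "MantleXTimes" quest.
; On activation it waits for a configured gameplay tag to be removed, then added again.
; It then increments and prints a count, sends a custom stat event to the owning player's quest manager, and ends.
; The count threshold check has empty branches, and failed passive activation also ends the ability.
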(event Ability|EventActivateAbility
  (bind _tag (Variables|Default|GetTag))
  (|CallParentFunction)
  (Ability|Tasks|WaitGameplayTagRemove _tag 0 true
    (:Removed
      (Ability|Tasks|WaitGameplayTagAdd _tag 0 true
        (:Added
          (bind _result (Math|Integer|IncrementInt (Variables|Default|GetCount))
            (:%20%20
              (Development|PrintString (Utilities|String|ToString(Integer) _result))
              (if (>= (Variables|Default|GetCount) 5)
                _
                (else
                  _))
              (Quest|SendComplexCustomStatEvent (Variables|GAVars|Data|GetOwningPlayerQuestManager) 0 "(GameplayTags=((TagName=\"Athena.Quests.MantleXTimes\")))" "(GameplayTags=)" 1)
              (Ability|EndAbility))))))))

(event Ability|EventFailedToActivatePassiveAbility
  (params (ActorInfo :type Gameplay%20Ability%20Actor%20Info%20Structure))
  (Ability|EndAbility))
